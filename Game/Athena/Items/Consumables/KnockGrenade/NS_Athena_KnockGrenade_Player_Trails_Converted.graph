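# Niagara player trail effect for the Knock Grenade, built from a single ribbon emitter named Body.
# The emitter spawns short-lived particles at a steady rate, with a skeletal mesh location module in particle update, rather than spawn, filtered to the spine_02 bone.
# Curves grow a scaled ribbon width and briefly raise then fade white alpha, alongside slight random lateral acceleration and constant dynamic material parameters.
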NiagaraSystem /Game/Athena/Items/Consumables/KnockGrenade/NS_Athena_KnockGrenade_Player_Trails_Converted
dependencies {
 usedDataInterfaces [&NiagaraDataInterfaceCurve, &NiagaraDataInterfaceSkeletalMesh]
 usedDynamicInputs []
 usedModules [&SystemState]
}
emitters {Body {
  emitterUpdateScript [
   &EmitterState(Inactive Response=::NewEnumerator1, Life Cycle Mode=::NewEnumerator1, Loop Behavior=::NewEnumerator1, Loop Duration=1, UseLoopDelay=true)
   &SpawnRate(SpawnRate=50, Use Spawn Probability=false) {module=SpawnRate001}
  ]
  particleSpawnScript [
   /SetVariables(Particles.Acceleration:Vector3f=&RandomRangeVector(Minimum=<-3 -3 0>, Maximum=<3 3 0>)) {isSetParameters=true}
   &InitializeParticle(Lifetime=0.4, Ribbon Width=20, Ribbon Width Mode=::NewEnumerator1, UsePositionOffset=false, Write Material Random=true)
  ]
  particleUpdateScript [
   /SetVariables(Particles.ScaledRibbonWidth:NiagaraFloat=&Multiply_Float(
     A=$Particles.RibbonWidth
     B=&FloatFromCurve(FloatCurve=DataInterface() {propertyValues {Curve {keys [
          <at -0.0035 -0.0003 RCIM_Constant>
          {arriveTangent=10.4171, interpMode=RCIM_Cubic, leaveTangent=10.4171, time=0.0709, value=0.0294}
          {arriveTangent=4.2588, interpMode=RCIM_Cubic, leaveTangent=4.2588, time=0.1413, value=1.5081}
          {interpMode=RCIM_Cubic, leaveTangent=0.4859, time=1, value=3.9862}
         ]}}})
    )) {isSetParameters=true}
   &SkeletalMeshLocation(
    Bone Sampling Mode=::NewEnumerator4
    MaskSpawnGroup=false
    Skeletal Mesh=DataInterface() {propertyValues {
      DefaultMesh=null
      ExcludeBoneName=root
      OptionalComponentClass=null
      PreviewMesh=/Game/Characters/Player/Female/Medium/Bodies/F_MED_Default_Soldier/Meshes/F_MED_Default_Soldier
      SkinningMode=SkinOnTheFly
      SoftSourceActor=null
      SourceMode=Default
      UvSetIndex=0
      WholeMeshLOD=-1
      bExcludeBone=false
      bReadDeformedGeometry=true
      bRequireCurrentFrameData=true
      ComponentTags []
      FilteredBones [spine_02]
      FilteredSockets []
      MeshUserParameter {parameter {name=None, typeDefHandle {registeredTypeIndex=1}, varData []}}
      SamplingRegions []
     }}
   )
   &ParticleState()
   &Color(Color=&MakeLinearColorFromVectorAndFloat(
     "Vector (RGB)"=<1 1 1>
     "Float (Alpha)"=&ClampFloat(Float=&FloatFromCurve(FloatCurve=DataInterface() {propertyValues {Curve {keys [<at 0.1 0 RCIM_Cubic>,<at 0.1984 0.25 RCIM_Cubic>,<at 1 0 RCIM_Cubic>]}}}))
    ))
   &AccelerationForce(Acceleration=$Particles.Acceleration)
   &SolveForcesAndVelocity(Clamp Velocity=false, Limit Acceleration=false)
   &DynamicMaterialParameters(
    Index 0 Param 1=0.5
    Index 0 Param 2=0.25
    Index 0 Param 3=1
    Index 0 Param 4=1
    Module.Index0Param0Write=true
    Module.Index0Param1Write=true
    Module.Index0Param2Write=true
    Module.Index0Param3Write=true
   )
  ]
  rendererClasses [&NiagaraRibbonRendererProperties]
  renderers [{rendererClass=&NiagaraRibbonRendererProperties}]
 }}
summary {
 systemName=NS_Athena_KnockGrenade_Player_Trails_Converted
 emitters [{emitterName=Body, rendererClasses [&NiagaraRibbonRendererProperties]}]
 userVariables []
}
